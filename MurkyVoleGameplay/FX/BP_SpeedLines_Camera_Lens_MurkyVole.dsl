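; Speed-lines camera lens effect Blueprint that conditionally switches on its Niagara effect.
; On BeginPlay it takes the dot product of player pawn 0's forward vector and its Local X Fwd variable, then clamps the result.
; A result above 0.5 activates the Niagara component without a reset.
; The check runs only once, as Tick and the construction script are empty.
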(event EventBeginPlay
  (if (> (Math|Float|Clamp(Float) (Math|Vector|DotProduct (Transformation|GetActorForwardVector (Game|GetPlayerPawn 0)) (Variables|Default|GetLocal%20X%20Fwd))) 0.5)
    (Components|Activation|Activate (Variables|NiagaraActor|GetNiagaraComponent) false)))

(event EventTick
  (params (DeltaSeconds :type Float%20%28single-precision%29)))

(fn ConstructionScript)
