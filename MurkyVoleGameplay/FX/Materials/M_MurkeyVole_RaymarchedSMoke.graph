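# Translucent, unlit, two-sided Niagara sprite material for lit flipbook smoke.
# A custom HLSL 2D raymarch steps the texture's green channel toward the atmospheric light for self-shadowing, with time-of-day color and density.
# Backlighting and a flattened sphere-normal term shape the lighting; a static switch replaces particle color with a noise-distorted spherical glow around the actor.
# Dynamic parameters drive noise-based erosion for opacity, depth-faded except on mobile.
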Material /MurkyVoleGameplay/FX/Materials/M_MurkeyVole_RaymarchedSMoke
graph {
 properties {
  bAutomaticallySetUsageInEditor=false
  bUsedWithNiagaraSprites=true
  blendMode=BLEND_Translucent
  pixelDepthOffsetMode=PDOM_Legacy
  shadingModel=MSM_Unlit
  twoSided=true
  usedShadingModels=Unlit
 }
 nodes [
  Add()
  Add()
  CameraVectorWS()
  DynamicParameter() {paramNames [Erode Power, Contrast, Erode Smoke, Shadow Density]}
  LocalPosition()
  LocalPosition()
  MaterialFunctionCall() {materialFunction=/Game/Effects/Fort_Effects/Materials/_Functions/CurveColor_Time_of_Day_S23}
  Multiply()
  Multiply()
  n1 = Multiply(A=TextureCoordinate(), B=ScalarParameter() {defaultValue=1.25, parameterName=Param_2})
  MaterialExpressionTextureObjectParameter_0 = TextureObjectParameter() {parameterName=TextureObject, samplerType=SAMPLERTYPE_Masks, texture=/Game/Effects/Niagara/MasterMaterials/Smoke/Textures/NEW/RT4x4_Combined}
  n2 = LinearInterpolate(
   B=LinearInterpolate(
    A=DynamicParameter() {paramNames [Erode Power, Contrast, Erode Smoke, Shadow Density]}.Erode Smoke
    Alpha=TextureSample(UVs=@n1) {texture=&2ndInfectedMapLinear}.G
   )
   Alpha=ParticleSubUV!(UVs, Tex=@MaterialExpressionTextureObjectParameter_0, Apply View MipBias) {samplerType=SAMPLERTYPE_LinearColor}.G
  )
  n3 = Power(Base=@n2, Exp=DynamicParameter() {paramNames [Erode Power, Contrast, Erode Smoke, Shadow Density]}.Erode Power)
  n4 = ParticleColor()
  Multiply(
   A=@n4.A
   B=MaterialFunctionCall(Alpha Threshold=@n3, Random=ParticleRandom()) {materialFunction=&DitherTemporalAA}.Result
  )
  Multiply() {constB=-1}
  n5 = TextureCoordinate() {uTiling=8, vTiling=8}
  n6 = DotProduct(
   A=MaterialFunctionCall(
    Normal=TextureSample(UVs=@n5) {samplerType=SAMPLERTYPE_Normal, texture=&T_Sphere_N}.RGB
    Flatness=ScalarParameter() {defaultValue=0.9, parameterName=Flatten Sphere Normal}
   ) {materialFunction=&FlattenNormal}.Result
   B=Transform(None=AtmosphericLightVector()) {transformSourceType=TRANSFORMSOURCE_World, transformType=TRANSFORM_Tangent}
  )
  NamedRerouteDeclaration(None=Power(Base=Saturate(None=@n6), Exp=ScalarParameter() {defaultValue=3})) {name=World Normal, nodeColor=<rgba 1 0.4688 0 1>}
  OneMinus()
  n7 = VectorParameter() {parameterName=Noise Tiling, defaultValue=<rgba 8 8 0 0>}
  n8 = TextureSample(UVs=Multiply(A=TextureCoordinate(), B=ComponentMask(None=@n7.RGB) {g=true, r=true})) {texture=/Game/Effects/FogVolume/noise_flamy_Tile}
  n9 = Add(A=WorldPosition().XYZ, B=Multiply(A=OneMinus(None=@n8.G), B=ScalarParameter() {defaultValue=250, parameterName=Glow distort}))
  n10 = ActorPositionWS()
  n11 = ScalarParameter() {defaultValue=750, parameterName=Glow Radius}
  n12 = ScalarParameter() {defaultValue=0.5, parameterName=Glow Hardness}
  n13 = Multiply(
   A=Multiply(
    A=SphereMask(A=@n9, B=@n10, Radius=@n11, Hardness=@n12)
    B=VectorParameter() {parameterName=Flame Glow, defaultValue=<rgba 2 0.1667 0 1>}.RGB
   )
   B=@n2
  )
  n14 = Multiply(A=@n2, B=@n4.RGB)
  MaterialExpressionScalarParameter_2 = ScalarParameter() {defaultValue=3, parameterName=MarchDistance}
  MaterialExpressionScalarParameter_1 = ScalarParameter() {defaultValue=16, parameterName=NumSteps}
  MaterialExpressionConstant_2 = Constant() {r=1}
  MaterialExpressionConstant3Vector_0 = Constant3Vector()
  MaterialExpressionMultiply_4 = Multiply(
   A=ScalarParameter() {defaultValue=8, parameterName=Overall Density}
   B=MaterialFunctionCall(
    Value_Morning=ScalarParameter() {defaultValue=2.5, parameterName=Density Scale Morning}
    Value_Day=ScalarParameter() {defaultValue=2.5, parameterName=Density Scale Day}
    Value_Evening=ScalarParameter() {defaultValue=2, parameterName=Density Scale Eveneing}
    Value_Night=ScalarParameter() {defaultValue=1.25, parameterName=Density Scale Night}
   ) {materialFunction=&Scalar_Time_of_Day}.Result
  )
  MaterialExpressionTextureCoordinate_0 = TextureCoordinate()
  MaterialExpressionTransform_0 = Transform(None=AtmosphericLightVector()) {transformSourceType=TRANSFORMSOURCE_World, transformType=TRANSFORM_Tangent}
  MaterialExpressionDesaturation_0 = Desaturation(
   None=MaterialFunctionCall(
    Color_Morning=VectorParameter() {parameterName="Color_Morning (V3)", defaultValue=<rgba 0.6062 0.3965 0.3372 1>}.RGB
    Color_Day=VectorParameter() {parameterName="Color_Day (V3)", defaultValue=<rgba 0.3479 0.28 0.2385 1>}.RGB
    Color_Evening=VectorParameter() {parameterName="Color_Evening (V3)", defaultValue=<rgba 0.5771 0.344 0.2721 1>}.RGB
    Color_Night=VectorParameter() {parameterName="Color_Night (V3)", defaultValue=<rgba 0.2498 0.294 0.5146 1>}.RGB
   ) {materialFunction=&Color_Time_of_Day}.Result
   Fraction=ScalarParameter() {defaultValue=0.75, parameterName=TOD Desat}
  )
  MaterialExpressionScalarParameter_6 = ScalarParameter() {defaultValue=4, parameterName=Contrast}
  n15 = Custom!(
   Tex=@MaterialExpressionTextureObjectParameter_0
   marchDistance=@MaterialExpressionScalarParameter_2
   numSteps=@MaterialExpressionScalarParameter_1
   TL=@MaterialExpressionConstant_2
   LO=@MaterialExpressionConstant3Vector_0
   density=@MaterialExpressionMultiply_4
   UV=@MaterialExpressionTextureCoordinate_0
   marchDir=@MaterialExpressionTransform_0
   LightColor=@MaterialExpressionDesaturation_0
   contrast=@MaterialExpressionScalarParameter_6
  ) {
   code="for (float d = 0.0; d < marchDistance; d += marchDistance / numSteps)\r\n{\r\n    UV += (marchDir * d) / numSteps;\r\n    float4 textureSample = Texture2DSampleLevel(Tex, TexSampler, UV.xy, 0) / (numSteps - contrast);\r\n    // Extracting the red channel\r\n    float densityChannel = textureSample.g;\r\n    TL *= saturate(1.0 - densityChannel);\r\n}\r\nfloat3 LI = LightColor * TL.xxx;\r\nLO += LI * density;\r\nreturn LO;"
   description=2D RayMarching
   inputs [
    {inputName=Tex, input {expression=&self:MaterialExpressionTextureObjectParameter_0}}
    {inputName=marchDistance, input {expression=&self:MaterialExpressionScalarParameter_2}}
    {inputName=numSteps, input {expression=&self:MaterialExpressionScalarParameter_1}}
    {inputName=TL, input {expression=&self:MaterialExpressionConstant_2}}
    {
     inputName=LO
     input {
      expression=&self:MaterialExpressionConstant3Vector_0
      mask=1
      maskB=1
      maskG=1
      maskR=1
     }
    }
    {inputName=density, input {expression=&self:MaterialExpressionMultiply_4}}
    {inputName=UV, input {expression=&self:MaterialExpressionTextureCoordinate_0}}
    {inputName=marchDir, input {expression=&self:MaterialExpressionTransform_0}}
    {inputName=LightColor, input {expression=&self:MaterialExpressionDesaturation_0}}
    {inputName=contrast, input {expression=&self:MaterialExpressionScalarParameter_6}}
   ]
  }
  n16 = Add(A=@n15, B=ScalarParameter() {parameterName=Param_4})
  n17 = Power(Base=@n15, Exp=ScalarParameter() {defaultValue=2, parameterName=Backlight Power})
  n18 = AtmosphericLightVector()
  n19 = Saturate(None=DotProduct(A=Multiply(A=@n18) {constB=-1}, B=VertexNormalWS()))
  n20 = Multiply(
   A=Multiply(A=Multiply(A=@n17, B=@n19), B=@MaterialExpressionDesaturation_0)
   B=ScalarParameter() {defaultValue=5, parameterName=Backlight Scale}
  )
  n21 = Subtract(
   A=Multiply(A=StaticSwitchParameter(True=@n13, False=@n14) {parameterName=Use Sphereical Glow}, B=Add(A=@n16, B=@n20))
   B=TextureSample() {texture=/Game/Effects/Niagara/MasterMaterials/Smoke/Textures/NEW/RT4x4_5_Tex}.G
  )
  n22 = ScalarParameter() {defaultValue=0.6, parameterName=Param_1}
  QualitySwitch(Default=LinearInterpolate(A=Multiply(A=@n21, B=@n22), B=@n21, Alpha=NamedRerouteUsage()), Low=@n21, Medium=@n21)
  ScalarParameter() {defaultValue=8, parameterName=Param_3}
  n23 = Multiply(A=@n4.A, B=@n3)
  ShadingPathSwitch(Default=DepthFade(Opacity=@n23), Mobile=@n23)
  Transform(None=VectorParameter() {defaultValue=<rgba 1 0 0 0>}.RGB) {transformSourceType=TRANSFORMSOURCE_World, transformType=TRANSFORM_Local}
  VectorParameter() {parameterName=Param_3, defaultValue=<rgba 0 0 0 50>}
 ]
}
