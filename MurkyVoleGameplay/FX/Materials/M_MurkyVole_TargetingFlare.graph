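# Additive, unlit, two-sided material for a targeting flare, drawn as a camera-facing sprite from a spikeball flare texture.
# A timestamp in/out transition, sharpened by a power, boosts sprite scale and HDR intensity of the custom-primitive-data color.
# World position offset combines sprite expansion with a camera offset.
# Named reroutes define a view-angle fade from local X against the camera vector and a sine/cosine flicker, alongside power-shaped size and opacity falloffs.
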Material /MurkyVoleGameplay/FX/Materials/M_MurkyVole_TargetingFlare
graph {
 properties {bAutomaticallySetUsageInEditor=false, blendMode=BLEND_Additive, shadingModel=MSM_Unlit, twoSided=true, usedShadingModels=Unlit}
 nodes [
  n1 = ScalarParameter() {parameterName=Time Stamp A}
  n2 = ScalarParameter() {defaultValue=1, parameterName=Time Stamp Fade In}
  n3 = ScalarParameter() {defaultValue=1, parameterName=Time Stamp B}
  n4 = ScalarParameter() {defaultValue=1, parameterName=Time Stamp Fade Out}
  n5 = Time()
  n6 = StaticBoolParameter() {parameterName="Transition In Only? (SB)"}
  n7 = Power(
   Base=Saturate(None=MaterialFunctionCall(TimestampA=@n1, TransitionLength_In=@n2, TimestampB=@n3, TransitionLength_Out=@n4, Time=@n5, Transition In Only?=@n6) {materialFunction=&MF_Timestamp_InOut_01}.SmoothStep)
   Exp=ScalarParameter() {defaultValue=6, parameterName=Param_2}
  )
  n8 = LinearInterpolate(Alpha=@n7) {constA=1, constB=2}
  n9 = AppendVector(A=ScalarParameter() {defaultValue=25, parameterName=X Scale}, B=ScalarParameter() {defaultValue=25, parameterName=Y Scale})
  n10 = NamedRerouteUsage()
  n11 = Multiply(A=@n9, B=Power(Base=LinearInterpolate(Alpha=@n10), Exp=ScalarParameter() {defaultValue=4, parameterName=WPO Fade}))
  n12 = ScalarParameter() {defaultValue=100}
  n13 = NamedRerouteUsage()
  Add(
   A=MaterialFunctionCall!(Center Location=WorldPosition().XYZ, 0-1 UVs=TextureCoordinate(), XY scale=Multiply(A=@n8, B=@n11), OverridePivot, "Normals (Optional)") {materialFunction=&Sprite}.World position offset
   B=MaterialFunctionCall(Offset Amount=LinearInterpolate(B=@n12, Alpha=@n13)) {materialFunction=/Engine/Functions/Engine_MaterialFunctions02/WorldPositionOffset/CameraOffset}.Camera Offset
  )
  n14 = VectorParameter() {bUseCustomPrimitiveData=true, parameterName=Color, primitiveDataIndex=4, defaultValue=<rgba 0 0.1168 1 1>}
  n15 = ScalarParameter() {defaultValue=15, parameterName=HDR}
  n16 = LinearInterpolate(Alpha=@n7) {constA=1, constB=4}
  MaterialFunctionCall(Emissive=MaterialFunctionCall(LDR Input=Saturate(None=@n14.RGB), HDR Tint=@n14.RGB, HDR Intensity Multiplier=Multiply(A=@n15, B=@n16)) {materialFunction=&DeriveHDRfromLDR}.RemappedRange) {materialFunction=&MF_FX_EmissiveAdjustments}
  n17 = NamedRerouteUsage()
  n18 = ScalarParameter() {defaultValue=8, parameterName=Param_1}
  Multiply(
   A=Multiply(A=Power(Base=@n17, Exp=@n18), B=TextureSample() {texture=/Game/Effects/Fort_Effects/Textures/Flares/T_Flare_Spikeball}.G)
   B=ScalarParameter() {defaultValue=1, parameterName=Overall Opacity}
  )
  Multiply(A=NamedRerouteUsage(), B=@n9)
  Multiply(B=ScalarParameter() {defaultValue=1, parameterName=Time Scale})
  n19 = VectorParameter() {defaultValue=<rgba 1 0 0 0>}
  n20 = Saturate(None=DotProduct(
    A=Transform(None=@n19.RGB) {transformSourceType=TRANSFORMSOURCE_Local}
    B=MaterialFunctionCall!(Use World Position Offsets) {materialFunction=/Engine/Functions/Engine_MaterialFunctions02/Utility/CameraVectorWithWPOOptions}.Result
   ))
  NamedRerouteDeclaration(None=LinearInterpolate(A=ScalarParameter() {defaultValue=1, parameterName=Min Scale}, B=ScalarParameter() {defaultValue=1, parameterName=Max Scale}, Alpha=@n20) {constA=0.5}) {name=Fade Along X, nodeColor=<rgba 1 0.2344 0 1>}
  n21 = Multiply(A=Time(), B=ScalarParameter() {defaultValue=1, parameterName=Flicker Time Scale})
  n22 = Add(A=Sine(None=@n21), B=Cosine(None=@n21))
  NamedRerouteDeclaration(None=LinearInterpolate(
    A=ScalarParameter() {defaultValue=0.5, parameterName=Flicker Min}
    B=ScalarParameter() {defaultValue=1, parameterName=Flicker Max}
    Alpha=Saturate(None=@n22)
   )) {name=Flicker, nodeColor=<rgba 1 0.6328 0 1>}
  NamedRerouteDeclaration(None=@n20) {name="Fade Along X~ No Lerp", nodeColor=<rgba 0.3125 0 1 1>}
  NamedRerouteUsage()
  OneMinus()
  OneMinus()
 ]
}
