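# Additive, unlit, two-sided material for the MurkyVole targeting laser beam.
# A primitive-data-driven Color, shaped by UV-derived gradients, is boosted to HDR and passed through the shared FX emissive adjustments.
# Opacity combines gradient masks, including a soft-edge falloff across U and a power-curve fade along the length, with an overall multiplier.
# SplineThicken supplies world position offset, with separate base and tip widths.
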Material /MurkyVoleGameplay/FX/Materials/M_MurkyVole_TargetingLaser
graph {
 properties {bAutomaticallySetUsageInEditor=false, blendMode=BLEND_Additive, shadingModel=MSM_Unlit, twoSided=true, usedShadingModels=Unlit}
 nodes [
  n1 = VectorParameter() {bUseCustomPrimitiveData=true, parameterName=Color, primitiveDataIndex=8, defaultValue=<rgba 0 0.0927 1 1>}
  n2 = NamedRerouteUsage()
  n3 = Multiply(A=Multiply(A=@n1.RGB, B=@n2), B=NamedRerouteUsage())
  MaterialFunctionCall(Emissive=MaterialFunctionCall(LDR Input=Saturate(None=@n3), HDR Tint=@n3, HDR Intensity Multiplier=ScalarParameter() {defaultValue=20, parameterName=HDR}) {materialFunction=&DeriveHDRfromLDR}.RemappedRange) {materialFunction=&MF_FX_EmissiveAdjustments}
  n4 = NamedRerouteUsage()
  n5 = ScalarParameter() {defaultValue=4, parameterName=Length Fade}
  Multiply(A=Multiply(A=@n2, B=Power(Base=@n4, Exp=@n5)), B=ScalarParameter() {defaultValue=1, parameterName=Overall Opacity})
  n6 = ComponentMask(None=TextureCoordinate()) {r=true}
  NamedRerouteDeclaration(None=Multiply(A=@n6, B=OneMinus(None=@n6))) {name=Soften Edges, nodeColor=<rgba 1 0 0.4922 1>}
  NamedRerouteDeclaration(None=ComponentMask(None=TextureCoordinate()) {g=true}) {name=Length Gradient, nodeColor=<rgba 0 0.8125 1 1>}
  NamedRerouteDeclaration(None=MaterialFunctionCall(WidthBase=ScalarParameter() {defaultValue=1, parameterName=Base}, WidthTip=ScalarParameter() {defaultValue=2.5, parameterName=Tip}) {materialFunction=&SplineThicken}.WorldPositionOffset) {name=WPO, nodeColor=<rgba 0 1 0.2266 1>}
  NamedRerouteUsage()
 ]
}
