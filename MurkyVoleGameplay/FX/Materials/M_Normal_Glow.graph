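# Additive unlit Niagara sprite material for a soft glow that responds to the normals of the scene behind it.
# A centered radial mask, raised to a power driven by particle alpha and multiplied by that alpha, shapes the falloff.
# Particle color is scaled by the inverted, saturated dot product of normalized particle world position and SceneTexture world normal.
# The mobile shading path skips the normal lookup and uses plain particle color.
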Material /MurkyVoleGameplay/FX/Materials/M_Normal_Glow
graph {
 properties {
  bAutomaticallySetUsageInEditor=false
  bUsedWithNiagaraSprites=true
  blendMode=BLEND_Additive
  shadingModel=MSM_Unlit
  usedShadingModels=Unlit
 }
 nodes [
  n1 = ParticleColor()
  n2 = TextureCoordinate()
  n3 = Constant() {r=0.5}
  Multiply(A=@n1.A, B=Power(Base=SphereMask(A=@n2, B=@n3) {attenuationRadius=0.25, hardnessPercent=0}, Exp=LinearInterpolate(Alpha=@n1.A) {constA=3}))
  Multiply()
  n4 = ParticlePositionWS()
  n5 = Saturate(None=DotProduct(A=Normalize(VectorInput=@n4), B=SceneTexture() {sceneTextureId=PPI_WorldNormal}.Color))
  ShadingPathSwitch(Default=Multiply(A=OneMinus(None=@n5), B=@n1.RGB), Mobile=@n1.RGB) {desc=Migrated from FeatureLevelSwitch}
 ]
}
