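# Niagara effect for the MurkyVole hover jetpack boost, attached to the backpack mesh's two thruster and two contrail sockets.
# Mesh flame layers and burst, flare sprites with lights, and curl-noise sparks form the exhaust, while sprites and ribbons form contrails.
# User.FuelFlicker blends color, size and waveform alpha flicker across most layers.
# User.uPhase scales the burst, and User.IsLocal tunes contrail lifetime, velocity and opacity.
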NiagaraSystem /MurkyVoleGameplay/FX/Systems/NS_MurkyVole_HoverJetBoost
dependencies {
 usedDataInterfaces [
  &NiagaraDataInterfaceColorCurve
  &NiagaraDataInterfaceVectorCurve
  &NiagaraDataInterfaceCurve
  &NiagaraDataInterfaceArrayFloat
 ]
 usedDynamicInputs []
 usedModules [&SystemState]
}
emitters {
 Airglow_Light_Flare {
  emitterUpdateScript [
   &EmitterState()
   &SpawnRate(SpawnRate=24, Use Spawn Probability=false)
  ]
  particleSpawnScript [
   &PartitionParticles()
   &InitializeParticle(
    Color=<rgba 1 1 1 0.25>
    Lifetime=0.17
    Sprite Size Mode=::NewEnumerator3
    Uniform Sprite Size=75
    UsePositionOffset=true
    Write Material Random=true
   )
   &CameraOffset()
  ]
  particleUpdateScript [
   &ParticleState()
   &ScaleSpriteSize(Uniform Curve Sprite Scale=DataInterface() {propertyValues {Curve {keys [{interpMode=RCIM_Cubic, leaveTangent=1.5808, tangentMode=RCTM_Break, time=0, value=0.5}, {tangentMode=RCTM_User, time=1, value=1}]}}})
   &FN_Sprite_Attributes(
    Color=&MultiplyLinearColor(
     Linear Color=&ColorFromCurve(
      ColorCurve=DataInterface() {propertyValues {
        AlphaCurve {keys [<at 1 0.3 RCIM_Cubic>]}
        BlueCurve {keys [<at 0 0>,<at 0.15 0>,<at 1 0>]}
        GreenCurve {keys [<at 0 0.0833>,<at 0.15 0.1422>,<at 1 0.25>]}
        RedCurve {keys [<at 0 1>,<at 0.15 4>,<at 1 1>]}
       }}
      CurveIndex=$User.FuelFlicker
     )
     Scale Factor=<1 1 1 1>
    )
    Create Camera Offset Attribute=true
    Create Color Attribute=true
    "Create Normalized Age "=true
    Create SpriteSize Attribute=true
    SpriteSize=&Multiply_Vector2D_ByFloat(
     Float=&FloatFromCurve(FloatCurve=DataInterface() {propertyValues {Curve {keys [
          {interpMode=RCIM_Cubic, leaveTangent=13.149, tangentMode=RCTM_Break, time=0, value=0}
          {interpMode=RCIM_Cubic, tangentMode=RCTM_Break, time=0.2, value=1.1}
          <at 1 0>
         ]}}})
     Vector2D=&Lerp_Vector2(A=<25 25>, B=<75 75>, Alpha=$User.FuelFlicker)
    )
    Write to Module Namespace=true
   )
   &FN_Light_Attributes(
    Color=&ColorFromCurve(
     ColorCurve=DataInterface() {propertyValues {
       AlphaCurve {keys [<at 1 0.3 RCIM_Cubic>]}
       BlueCurve {keys [<at 0.15 0>,<at 0.561 0>,<at 1 2.5 RCIM_Cubic>]}
       GreenCurve {keys [<at 0.15 0.1422>,<at 0.561 0.7033>,<at 1 1.6592 RCIM_Cubic>]}
       RedCurve {keys [<at 0.15 4>,<at 0.561 1>,<at 1 0 RCIM_Cubic>]}
      }}
     CurveIndex=$User.FuelFlicker
    )
    Create Color Attribute=true
    Create Exponent Attribute=true
    Create Radius Attribute=true
    Enable Particle Light Enable=true
    Exponent=4
    Light Enabled=true
    Radius=&FloatFromCurve(
     FloatCurve=DataInterface() {propertyValues {Curve {keys [
         {arriveTangent=9.0725, interpMode=RCIM_Cubic, leaveTangent=9.0725, tangentMode=RCTM_User, time=0, value=0}
         {arriveTangent=-2.5249, interpMode=RCIM_Cubic, leaveTangent=-2.5249, tangentMode=RCTM_User, time=0.1408, value=0.9847}
         {arriveTangent=-0.0405, interpMode=RCIM_Cubic, leaveTangent=-0.0405, tangentMode=RCTM_User, time=1, value=0}
        ]}}}
     Scale Curve=200
    )
   )
   &FN_Sprite_Attributes(
    Color=&MultiplyLinearColor(
     Linear Color=&ColorFromCurve(
      ColorCurve=DataInterface() {propertyValues {
        AlphaCurve {keys [<at 1 0.3 RCIM_Cubic>]}
        BlueCurve {keys [<at 0 0>,<at 0.15 0>,<at 1 0>]}
        GreenCurve {keys [<at 0 0.0833>,<at 0.15 0.1422>,<at 1 0.25>]}
        RedCurve {keys [<at 0 1>,<at 0.15 4>,<at 1 1>]}
       }}
      CurveIndex=$User.FuelFlicker
     )
     Scale Factor=<1 1 1 1>
    )
    Create Camera Offset Attribute=true
    Create Color Attribute=true
    Create SpriteSize Attribute=true
    SpriteSize=&Multiply_Vector2D_ByFloat(
     Float=&RandomRangeFloat(Minimum=0.9, Maximum=1, Evaluation Type=::NewEnumerator1)
     Vector2D=&Lerp_Vector2(A=<31 2.5>, B=<250 5>, Alpha=$User.FuelFlicker)
    )
    Write to Module Namespace=true
   ) {module=FN_Sprite_Attributes001}
   &ScaleColor(
    Curve Index=$User.FuelFlicker
    Linear Color Curve=DataInterface() {propertyValues {
      AlphaCurve {keys [<at 1 0.3 RCIM_Cubic>]}
      BlueCurve {keys [<at 0 0>,<at 0.15 0>,<at 1 0>]}
      GreenCurve {keys [<at 0 0.1333>,<at 0.15 0.6667>,<at 1 0.25>]}
      RedCurve {keys [<at 0 1>,<at 0.15 4>,<at 1 1>]}
     }}
    Scale Mode=::NewEnumerator2
   )
   &SkeletalMeshLocation(
    Bone / Socket Index=$Particles.PartitionParticles.Partition
    MaskSpawnGroup=false
    Mesh Sampling Type=::NewEnumerator5
    Sampled Position Offset=<8 0 0>
    Skeletal Mesh=DataInterface() {propertyValues {
      DefaultMesh=/MurkyVoleGameplay/Assets/Backpack/Meshes/M_MED_MurkyVole_Pack
      ExcludeBoneName=root
      OptionalComponentClass=null
      PreviewMesh=&M_MED_MurkyVole_Pack
      SkinningMode=SkinOnTheFly
      SoftSourceActor=null
      SourceMode=Default
      UvSetIndex=0
      WholeMeshLOD=-1
      bExcludeBone=false
      bReadDeformedGeometry=true
      bRequireCurrentFrameData=true
      ComponentTags []
      FilteredBones []
      FilteredSockets [Socket_JetThruster_Left, Socket_JetThruster_Right]
      MeshUserParameter {parameter {name=None, typeDefHandle {registeredTypeIndex=1}, varData []}}
      SamplingRegions []
     }}
    Socket Sampling Mode=::NewEnumerator5
   )
   &DynamicMaterialParameters(
    Index 0 Param 1=1
    Index 0 Param 2=5
    Index 0 Param 3=1
    Index 0 Param 4=1
    Module.Index0Param0Write=true
    Module.Index0Param1Write=true
    Module.Index0Param2Write=true
    Module.Index0Param3Write=true
   )
  ]
  rendererClasses [&NiagaraSpriteRendererProperties, &NiagaraLightRendererProperties]
  renderers [
   {rendererClass=&NiagaraSpriteRendererProperties}
   {rendererClass=&NiagaraSpriteRendererProperties, rendererIndex=1}
   {rendererClass=&NiagaraLightRendererProperties, rendererIndex=2}
   {rendererClass=&NiagaraSpriteRendererProperties, rendererIndex=3}
  ]
 }
 Burst {
  emitterUpdateScript [
   &EmitterState(Life Cycle Mode=::NewEnumerator1, Loop Behavior=::NewEnumerator1, Loop Duration=1, UseLoopDelay=false)
   &SpawnBurst_Instantaneous(Spawn Count=2, Use Spawn Probability=false, UseLoopCountLimit=false)
  ]
  particleSpawnScript [
   &PartitionParticles()
   &InitializeParticle(
    Color Mode=::NewEnumerator0
    Lifetime Max=0.25
    Lifetime Min=0.15
    Lifetime Mode=::NewEnumerator1
    Position Mode=::NewEnumerator0
    Write Material Random=true
   )
   /SetVariables(
    Particles.Color_Blaster:LinearColor=<rgba 20 1.6667 0 0.1>
    Particles.Orientation_Blaster:Quat4f=&OrientMeshToVector(Look At Direction=<0 -1 0>, Reference Axis Vector=<1 0 0>)
    Particles.Size_Blaster:Vector3f=<5 1 1>
   ) {isSetParameters=true, module=SetVariables_#1}
  ]
  particleUpdateScript [
   /SetVariables(
    Particles.Color_Blaster:LinearColor=&MultiplyLinearColor(
     Linear Color=&ColorFromCurve(ColorCurve=DataInterface() {propertyValues {
        AlphaCurve {keys [<at 0 1 RCIM_Cubic>,<at 1 0>]}
        BlueCurve {keys [<at 0.1065 3.5 RCIM_Cubic>,<at 0.5271 0>]}
        GreenCurve {keys [<at 0.1065 1.4518 RCIM_Cubic>,<at 0.5271 0.8>]}
        RedCurve {keys [<at 0.1065 0 RCIM_Cubic>,<at 0.5271 8>]}
       }})
     Scale Factor=&MakeFloat4FromVectorAndFloat(
      "Vector (RGB)"=&VectorFromCurve(VectorCurve=DataInterface() {propertyValues {
         XCurve {keys [<at 0 2 RCIM_Cubic>,<at 1 1 RCIM_Cubic>]}
         YCurve {keys [<at 0 2 RCIM_Cubic>,<at 1 1 RCIM_Cubic>]}
         ZCurve {keys [<at 0 2 RCIM_Cubic>,<at 1 1 RCIM_Cubic>]}
        }})
      "Float (Alpha)"=&Lerp_Float(
       A=&Waveform("Waveform [1]"=::NewEnumerator8, "[1] Frequency"=1, "[1] Amplitude Min/Max"=<0 1>, "[1] Scale"=2)
       B=&FloatFromCurve(
        FloatCurve=DataInterface() {propertyValues {Curve {keys [<at 0 0 RCIM_Cubic>,<at 0.5 1>,<at 1.0011 0 RCIM_Cubic>]}}}
        Scale Curve=&Lerp_Float(
         A=$User.FuelFlicker
         B=1
         Alpha=&Waveform("Waveform [1]"=::NewEnumerator8, "[1] Phase Scale"=50, "[1] Frequency"=5, "[1] Amplitude Min/Max"=<0 1>)
        )
       )
       Alpha=$User.FuelFlicker
      )
     )
    )
    Particles.DynamicParameter_Blaster:Vector4f=&MakeVector4(
     X=&FloatFromCurve(FloatCurve=DataInterface() {propertyValues {Curve {keys [<at 0 1>,<at 1 0>]}}})
     Y=5
     Z=20
     W=1
    )
    Particles.Size_Blaster:Vector3f=&Multiply_VectorByFloat(
     Vector=&Multiply_Vector(
      A=&VectorFromCurve(
       VectorCurve=DataInterface() {propertyValues {
         XCurve {keys [
           <at 0 0.75 RCIM_Cubic>
           <at 0.4908 3.9579>
           {arriveTangent=4.355, interpMode=RCIM_Cubic, leaveTangent=4.355, tangentMode=RCTM_User, time=0.5, value=4}
           <at 1 0.5 RCIM_Cubic>
          ]}
         YCurve {keys [
           <at 0 1 RCIM_Cubic>
           <at 0.4908 0.9933>
           {arriveTangent=0.75, interpMode=RCIM_Cubic, leaveTangent=0.75, time=0.5, value=1}
           <at 1 1.75 RCIM_Cubic>
          ]}
         ZCurve {keys [<at -0.0014 1 RCIM_Cubic>, {arriveTangent=0.7489, interpMode=RCIM_Cubic, leaveTangent=0.7489, time=0.5, value=1}, <at 1 1.75 RCIM_Cubic>]}
        }}
       Scale Curve=&MakeVector(
        X=&Lerp_Float(A=0, B=0.25, Alpha=$User.uPhase)
        Y=&Lerp_Float(A=0.15, B=0.35, Alpha=$User.uPhase)
        Z=&Lerp_Float(A=0.25, B=0.5, Alpha=$User.uPhase)
       )
      )
      B=&VectorFromFloat(Value=&FloatFromCurve(
        FloatCurve=DataInterface() {propertyValues {Curve {keys [
            {arriveTangent=5.1112, interpMode=RCIM_Cubic, leaveTangent=5.1112, tangentMode=RCTM_User, time=0, value=0}
            {arriveTangent=-4.9135, interpMode=RCIM_Cubic, leaveTangent=-4.9135, tangentMode=RCTM_User, time=1, value=0.2}
           ]}}}
        Scale Curve=&Lerp_Float(
         A=$User.FuelFlicker
         B=1
         Alpha=&Waveform("Waveform [1]"=::NewEnumerator8, "[1] Phase Scale"=50, "[1] Frequency"=5, "[1] Amplitude Min/Max"=<0 1>)
        )
       ))
     )
     Float=1
    )
   ) {isSetParameters=true, module=SetVariables_#2}
   &ParticleState()
   &SkeletalMeshLocation(
    Bone / Socket Index=$Particles.PartitionParticles.Partition
    MaskSpawnGroup=false
    Mesh Sampling Type=::NewEnumerator5
    Sampled Position Offset=<7 0 0>
    Skeletal Mesh=DataInterface() {propertyValues {
      DefaultMesh=&M_MED_MurkyVole_Pack
      ExcludeBoneName=root
      OptionalComponentClass=null
      PreviewMesh=&M_MED_MurkyVole_Pack
      SkinningMode=SkinOnTheFly
      SoftSourceActor=null
      SourceMode=Default
      UvSetIndex=0
      WholeMeshLOD=-1
      bExcludeBone=false
      bReadDeformedGeometry=true
      bRequireCurrentFrameData=true
      ComponentTags []
      FilteredBones []
      FilteredSockets [Socket_JetThruster_Left, Socket_JetThruster_Right]
      MeshUserParameter {parameter {name=None, typeDefHandle {registeredTypeIndex=1}, varData []}}
      SamplingRegions []
     }}
    Socket Sampling Mode=::NewEnumerator5
   )
   /SetVariables(Particles.MeshOrientation=$Particles.SkeletalMeshLocation.SampledRotation) {isSetParameters=true, module=SetVariables_#3}
  ]
  rendererClasses [&NiagaraMeshRendererProperties]
  renderers [{rendererClass=&NiagaraMeshRendererProperties, rendererIndex=0}]
 }
 Combined_Rate {
  emitterUpdateScript [
   &EmitterState()
   &SpawnRate(SpawnRate=24, Use Spawn Probability=false)
  ]
  particleSpawnScript [
   &PartitionParticles()
   &InitializeParticle(
    Color Mode=::NewEnumerator0
    Lifetime Max=0.35
    Lifetime Min=0.25
    Lifetime Mode=::NewEnumerator1
    Position Mode=::NewEnumerator0
    Write Material Random=true
   )
   /SetVariables(
    Particles.Color_Blaster:LinearColor=<rgba 20 1.6667 0 0.1>
    Particles.Orientation_Blaster:Quat4f=&OrientMeshToVector(Look At Direction=<0 -1 0>, Reference Axis Vector=<1 0 0>)
    Particles.Size_Blaster:Vector3f=<1 1 1>
   ) {isSetParameters=true, module=SetVariables_#1}
  ]
  particleUpdateScript [
   /SetVariables(
    Particles.Color_Blaster:LinearColor=&MultiplyLinearColor(
     Linear Color=&MultiplyLinearColor(
      Linear Color=&ColorFromCurve(
       ColorCurve=DataInterface() {propertyValues {
         AlphaCurve {keys [<at 1 0.3 RCIM_Cubic>]}
         BlueCurve {keys [<at 0 0>,<at 0.15 0>,<at 1 0>]}
         GreenCurve {keys [<at 0 0.0833>,<at 0.15 0.1422>,<at 1 0.25>]}
         RedCurve {keys [<at 0 1>,<at 0.15 4>,<at 1 1>]}
        }}
       CurveIndex=$User.FuelFlicker
      )
      Scale Factor=<2 2 2 1>
     )
     Scale Factor=&MakeFloat4FromVectorAndFloat(
      "Vector (RGB)"=&VectorFromCurve(VectorCurve=DataInterface() {propertyValues {XCurve {keys [<at 0 2 RCIM_Cubic>]}, YCurve {keys [<at 0 2 RCIM_Cubic>]}, ZCurve {keys [<at 0 2 RCIM_Cubic>]}}})
      "Float (Alpha)"=&Lerp_Float(
       A=&Waveform(
        Global Phase=$System.Age
        "Waveform [1]"=::NewEnumerator8
        "[1] Phase"=$System.Age
        "[1] Frequency"=50
        "[1] Amplitude Min/Max"=<-0.5 1>
        "[1] Scale"=2
       )
       B=&FloatFromCurve(FloatCurve=DataInterface() {propertyValues {Curve {keys [<at 0 0 RCIM_Cubic>,<at 0.5 1>,<at 1.0011 0 RCIM_Cubic>]}}})
       Alpha=$User.FuelFlicker
      )
     )
    )
    Particles.DynamicParameter_Blaster:Vector4f=&MakeVector4(
     X=&FloatFromCurve(FloatCurve=DataInterface() {propertyValues {Curve {keys [<at 0 1>,<at 1 0>]}}})
     Y=5
     Z=5
     W=1
    )
    Particles.Size_Blaster:Vector3f=&Multiply_Vector(
     A=&VectorFromCurve(
      VectorCurve=DataInterface() {propertyValues {
        XCurve {keys [<at 0 0.75 RCIM_Cubic>, {interpMode=RCIM_Cubic, tangentMode=RCTM_Break, time=0.4908, value=8}, <at 1 0.5 RCIM_Cubic>]}
        YCurve {keys [
          <at 0 1 RCIM_Cubic>
          <at 0.4908 0.9933>
          {arriveTangent=0.75, interpMode=RCIM_Cubic, leaveTangent=0.75, time=0.5, value=1}
          <at 1 1.75 RCIM_Cubic>
         ]}
        ZCurve {keys [<at -0.0014 1 RCIM_Cubic>, {arriveTangent=0.7489, interpMode=RCIM_Cubic, leaveTangent=0.7489, time=0.5, value=1}, <at 1 1.75 RCIM_Cubic>]}
       }}
      Scale Curve=&Lerp_Vector(A=<0.05 0.5 0.5>, B=<0.075 0.5 0.5>, Alpha=$User.FuelFlicker)
     )
     B=&VectorFromFloat(Value=&FloatFromCurve(FloatCurve=DataInterface() {propertyValues {Curve {keys [
           {arriveTangent=5.1112, interpMode=RCIM_Cubic, leaveTangent=5.1112, tangentMode=RCTM_User, time=0, value=0}
           {arriveTangent=-4.9135, interpMode=RCIM_Cubic, leaveTangent=-4.9135, tangentMode=RCTM_User, time=1, value=0.2}
          ]}}}))
    )
   ) {isSetParameters=true, module=SetVariables_#2}
   &ParticleState()
   &SkeletalMeshLocation(
    Bone / Socket Index=$Particles.PartitionParticles.Partition
    MaskSpawnGroup=false
    Mesh Sampling Type=::NewEnumerator5
    Sampled Position Offset=<6 0 0>
    Skeletal Mesh=DataInterface() {propertyValues {
      DefaultMesh=&M_MED_MurkyVole_Pack
      ExcludeBoneName=root
      OptionalComponentClass=null
      PreviewMesh=&M_MED_MurkyVole_Pack
      SkinningMode=SkinOnTheFly
      SoftSourceActor=null
      SourceMode=Default
      UvSetIndex=0
      WholeMeshLOD=-1
      bExcludeBone=false
      bReadDeformedGeometry=true
      bRequireCurrentFrameData=true
      ComponentTags []
      FilteredBones []
      FilteredSockets [Socket_JetThruster_Left, Socket_JetThruster_Right]
      MeshUserParameter {parameter {name=None, typeDefHandle {registeredTypeIndex=1}, varData []}}
      SamplingRegions []
     }}
    Socket Sampling Mode=::NewEnumerator5
   )
   /SetVariables(Particles.MeshOrientation=$Particles.SkeletalMeshLocation.SampledRotation) {isSetParameters=true, module=SetVariables_#3}
  ]
  rendererClasses [&NiagaraMeshRendererProperties]
  renderers [{rendererClass=&NiagaraMeshRendererProperties, rendererIndex=0}]
 }
 Combined_Rate001 {
  emitterUpdateScript [
   &EmitterState()
   &SpawnRate(SpawnRate=24, Use Spawn Probability=false)
  ]
  particleSpawnScript [
   &PartitionParticles()
   &InitializeParticle(
    Color Mode=::NewEnumerator0
    Lifetime Max=0.35
    Lifetime Min=0.25
    Lifetime Mode=::NewEnumerator1
    Position Mode=::NewEnumerator0
    Write Material Random=true
   )
   /SetVariables(
    Particles.Color_Blaster:LinearColor=<rgba 20 1.6667 0 0.1>
    Particles.Orientation_Blaster:Quat4f=&OrientMeshToVector(Look At Direction=<0 -1 0>, Reference Axis Vector=<1 0 0>)
    Particles.Size_Blaster:Vector3f=<1 1 1>
   ) {isSetParameters=true, module=SetVariables_#1}
  ]
  particleUpdateScript [
   /SetVariables(
    Particles.Color_Blaster:LinearColor=&MultiplyLinearColor(
     Linear Color=&MultiplyLinearColor(
      Linear Color=&ColorFromCurve(
       ColorCurve=DataInterface() {propertyValues {
         AlphaCurve {keys [<at 0 0>,<at 1 0.3 RCIM_Cubic>]}
         BlueCurve {keys [<at 0 0>,<at 0.15 2.5 RCIM_Cubic>]}
         GreenCurve {keys [<at 0 0>,<at 0.15 1.6592 RCIM_Cubic>]}
         RedCurve {keys [<at 0 0>,<at 0.15 0 RCIM_Cubic>]}
        }}
       CurveIndex=$User.FuelFlicker
      )
      Scale Factor=<4 4 4 1>
     )
     Scale Factor=&MakeFloat4FromVectorAndFloat(
      "Vector (RGB)"=&VectorFromCurve(VectorCurve=DataInterface() {propertyValues {XCurve {keys [<at 0 2 RCIM_Cubic>]}, YCurve {keys [<at 0 2 RCIM_Cubic>]}, ZCurve {keys [<at 0 2 RCIM_Cubic>]}}})
      "Float (Alpha)"=&Lerp_Float(
       A=&Waveform(
        Global Phase=$System.Age
        "Waveform [1]"=::NewEnumerator8
        "[1] Phase"=$System.Age
        "[1] Frequency"=50
        "[1] Amplitude Min/Max"=<-0.5 1>
        "[1] Scale"=2
       )
       B=&FloatFromCurve(FloatCurve=DataInterface() {propertyValues {Curve {keys [<at 0 0 RCIM_Cubic>,<at 0.5 1>,<at 1.0011 0 RCIM_Cubic>]}}})
       Alpha=$User.FuelFlicker
      )
     )
    )
    Particles.DynamicParameter_Blaster:Vector4f=&MakeVector4(
     X=&FloatFromCurve(FloatCurve=DataInterface() {propertyValues {Curve {keys [<at 0 1>,<at 1 0>]}}})
     Y=5
     Z=5
     W=1
    )
    Particles.Size_Blaster:Vector3f=&Multiply_Vector(
     A=&VectorFromCurve(
      VectorCurve=DataInterface() {propertyValues {
        XCurve {keys [<at 0 0.75 RCIM_Cubic>, {interpMode=RCIM_Cubic, tangentMode=RCTM_Break, time=0.4908, value=8}, <at 1 0.5 RCIM_Cubic>]}
        YCurve {keys [
          <at 0 1 RCIM_Cubic>
          <at 0.4908 0.9933>
          {arriveTangent=0.75, interpMode=RCIM_Cubic, leaveTangent=0.75, time=0.5, value=1}
          <at 1 1.75 RCIM_Cubic>
         ]}
        ZCurve {keys [<at -0.0014 1 RCIM_Cubic>, {arriveTangent=0.7489, interpMode=RCIM_Cubic, leaveTangent=0.7489, time=0.5, value=1}, <at 1 1.75 RCIM_Cubic>]}
       }}
      Scale Curve=&Lerp_Vector(A=<0.015 0.35 0.35>, B=<0.03 0.35 0.35>, Alpha=$User.FuelFlicker)
     )
     B=&VectorFromFloat(Value=&FloatFromCurve(FloatCurve=DataInterface() {propertyValues {Curve {keys [
           {arriveTangent=5.1112, interpMode=RCIM_Cubic, leaveTangent=5.1112, tangentMode=RCTM_User, time=0, value=0}
           {arriveTangent=-4.9135, interpMode=RCIM_Cubic, leaveTangent=-4.9135, tangentMode=RCTM_User, time=1, value=0.2}
          ]}}}))
    )
   ) {isSetParameters=true, module=SetVariables_#2}
   &ParticleState()
   &SkeletalMeshLocation(
    Bone / Socket Index=$Particles.PartitionParticles.Partition
    MaskSpawnGroup=false
    Mesh Sampling Type=::NewEnumerator5
    Sampled Position Offset=<6 0 0>
    Skeletal Mesh=DataInterface() {propertyValues {
      DefaultMesh=&M_MED_MurkyVole_Pack
      ExcludeBoneName=root
      OptionalComponentClass=null
      PreviewMesh=&M_MED_MurkyVole_Pack
      SkinningMode=SkinOnTheFly
      SoftSourceActor=null
      SourceMode=Default
      UvSetIndex=0
      WholeMeshLOD=-1
      bExcludeBone=false
      bReadDeformedGeometry=true
      bRequireCurrentFrameData=true
      ComponentTags []
      FilteredBones []
      FilteredSockets [Socket_JetThruster_Left, Socket_JetThruster_Right]
      MeshUserParameter {parameter {name=None, typeDefHandle {registeredTypeIndex=1}, varData []}}
      SamplingRegions []
     }}
    Socket Sampling Mode=::NewEnumerator5
   )
   /SetVariables(Particles.MeshOrientation=$Particles.SkeletalMeshLocation.SampledRotation) {isSetParameters=true, module=SetVariables_#3}
  ]
  rendererClasses [&NiagaraMeshRendererProperties]
  renderers [{rendererClass=&NiagaraMeshRendererProperties, rendererIndex=0}]
 }
 Flame {
  emitterUpdateScript [
   &EmitterState()
   &SpawnBurst_Instantaneous(Spawn Count=2, Use Spawn Probability=false, UseLoopCountLimit=false)
  ]
  particleSpawnScript [
   &PartitionParticles()
   &InitializeParticle(
    Color=<rgba 12.5 14.307 20 0.5>
    Lifetime=1.4
    Mesh Scale=<0.2227 0.0369 0.0369>
    Mesh Scale Mode=::NewEnumerator4
    UsePositionOffset=false
    Write Material Random=true
   )
   &DynamicMaterialParameters(
    Index 0 Param 2=5
    Index 0 Param 4=20
    Index 1 Param 2=20
    Module.Param0WriteEnabled=true
    Module.Param0WriteEnabled1=true
    Module.Param1WriteEnabled=true
    Module.Param1WriteEnabled1=true
    Module.Param2WriteEnabled=true
    Module.Param2WriteEnabled1=true
    Module.Param3WriteEnabled=true
    Module.Param3WriteEnabled1=true
    Write Parameter Index 1=true
   )
  ]
  particleUpdateScript [
   &ParticleState()
   &SkeletalMeshLocation(
    Bone / Socket Index=$Particles.PartitionParticles.Partition
    MaskSpawnGroup=false
    Mesh Sampling Type=::NewEnumerator5
    Sampled Position Offset=<7 0 0>
    Skeletal Mesh=DataInterface() {propertyValues {
      DefaultMesh=&M_MED_MurkyVole_Pack
      ExcludeBoneName=root
      OptionalComponentClass=null
      PreviewMesh=&M_MED_MurkyVole_Pack
      SkinningMode=SkinOnTheFly
      SoftSourceActor=null
      SourceMode=Default
      UvSetIndex=0
      WholeMeshLOD=-1
      bExcludeBone=false
      bReadDeformedGeometry=true
      bRequireCurrentFrameData=true
      ComponentTags []
      FilteredBones []
      FilteredSockets [Socket_JetThruster_Left, Socket_JetThruster_Right]
      MeshUserParameter {parameter {name=None, typeDefHandle {registeredTypeIndex=1}, varData []}}
      SamplingRegions []
     }}
    Socket Sampling Mode=::NewEnumerator5
   )
   &ScaleColor(
    Scale Alpha=&FloatFromCurve(
     FloatCurve=DataInterface() {propertyValues {Curve {keys [
         <at 0 1 RCIM_Cubic>
         {arriveTangent=0.3177, interpMode=RCIM_Cubic, leaveTangent=0.3177, time=0.0746, value=0.7169}
         {arriveTangent=-0.0227, interpMode=RCIM_Cubic, leaveTangent=-0.0227, tangentMode=RCTM_User, time=0.121, value=1.0384}
         {arriveTangent=-4.2288, interpMode=RCIM_Cubic, leaveTangent=-4.2288, time=0.1457, value=0.5744}
         {arriveTangent=2.1347, interpMode=RCIM_Cubic, leaveTangent=2.1347, time=0.1738, value=0.815}
         {arriveTangent=2.7457, interpMode=RCIM_Cubic, leaveTangent=2.7457, time=0.2093, value=0.7101}
         {arriveTangent=-2.342, interpMode=RCIM_Cubic, leaveTangent=-2.342, time=0.2396, value=0.9955}
         {arriveTangent=-4.6791, interpMode=RCIM_Cubic, leaveTangent=-4.6791, time=0.262, value=0.5867}
         {arriveTangent=0.0978, interpMode=RCIM_Cubic, leaveTangent=0.0978, time=0.2791, value=0.8105}
         {arriveTangent=0.9183, interpMode=RCIM_Cubic, leaveTangent=0.9183, time=0.3099, value=0.5914}
         {arriveTangent=2.3091, interpMode=RCIM_Cubic, leaveTangent=2.3091, time=0.3406, value=0.867}
         {arriveTangent=4.828, interpMode=RCIM_Cubic, leaveTangent=4.828, time=0.3796, value=0.7524}
         {arriveTangent=-2.2784, interpMode=RCIM_Cubic, leaveTangent=-2.2784, time=0.4112, value=1.2077}
         {arriveTangent=-7.8731, interpMode=RCIM_Cubic, leaveTangent=-7.8731, time=0.4339, value=0.6287}
         {arriveTangent=-1.4466, interpMode=RCIM_Cubic, leaveTangent=-1.4466, time=0.4676, value=0.7633}
         {arriveTangent=-1.1137, interpMode=RCIM_Cubic, leaveTangent=-1.1137, time=0.4986, value=0.5352}
         {arriveTangent=0.3293, interpMode=RCIM_Cubic, leaveTangent=0.3293, time=0.5283, value=0.6957}
         {arriveTangent=2.183, interpMode=RCIM_Cubic, leaveTangent=2.183, time=0.5598, value=0.5553}
         {arriveTangent=1.1553, interpMode=RCIM_Cubic, leaveTangent=1.1553, time=0.5824, value=0.8137}
         {arriveTangent=1.117, interpMode=RCIM_Cubic, leaveTangent=1.117, time=0.6056, value=0.6083}
         {arriveTangent=3.4815, interpMode=RCIM_Cubic, leaveTangent=3.4815, time=0.6363, value=0.874}
         {arriveTangent=2.1233, interpMode=RCIM_Cubic, leaveTangent=2.1233, time=0.6464, value=0.7503}
         {arriveTangent=2.4312, interpMode=RCIM_Cubic, leaveTangent=2.4312, time=0.6964, value=1.0015}
         {arriveTangent=1.4106, interpMode=RCIM_Cubic, leaveTangent=1.4106, time=0.7177, value=0.9236}
         {arriveTangent=-6.8551, interpMode=RCIM_Cubic, leaveTangent=-6.8551, time=0.7385, value=1.061}
         {arriveTangent=-0.0928, interpMode=RCIM_Cubic, leaveTangent=-0.0928, tangentMode=RCTM_User, time=0.7644, value=0.6029}
         {arriveTangent=1.9217, interpMode=RCIM_Cubic, leaveTangent=1.9217, time=0.7931, value=0.8896}
         {arriveTangent=1.6033, interpMode=RCIM_Cubic, leaveTangent=1.6033, time=0.8092, value=0.689}
         {arriveTangent=-0.3203, interpMode=RCIM_Cubic, leaveTangent=-0.3203, time=0.8924, value=1.0487}
         {arriveTangent=-6.4298, interpMode=RCIM_Cubic, leaveTangent=-6.4298, time=0.911, value=0.6564}
         {arriveTangent=3.8595, interpMode=RCIM_Cubic, leaveTangent=3.8595, time=0.9278, value=0.8209}
         {interpMode=RCIM_Cubic, leaveTangent=0.0463, time=1, value=1}
        ]}}}
     Scale Curve=&FloatFromCurve(FloatCurve=DataInterface() {propertyValues {Curve {keys [
          {tangentMode=RCTM_User, time=0, value=1}
          {arriveTangent=-0.0405, interpMode=RCIM_Cubic, leaveTangent=-0.0405, tangentMode=RCTM_User, time=0.7009, value=1}
          {arriveTangent=-8.0089, interpMode=RCIM_Cubic, leaveTangent=-8.0089, tangentMode=RCTM_User, time=1, value=0}
         ]}}})
    )
    ScaleA=true
    ScaleRGB=true
   )
   &ScaleMeshSize(Scale Factor=&MakeVector(
     X=&FloatFromCurve(FloatCurve=DataInterface() {propertyValues {Curve {keys [
          <at 0 0 RCIM_Cubic>
          {arriveTangent=2, interpMode=RCIM_Cubic, leaveTangent=2, time=0.25, value=1.7}
          {arriveTangent=-1.1667, interpMode=RCIM_Cubic, leaveTangent=-1.1667, time=0.5, value=1}
          {arriveTangent=-5.7307, interpMode=RCIM_Cubic, leaveTangent=-5.7307, tangentMode=RCTM_User, time=1, value=0.825}
         ]}}})
     Y=&FloatFromCurve(FloatCurve=DataInterface() {propertyValues {Curve {keys [<at 0 0.8 RCIM_Cubic>,<at 0.5 1.2 RCIM_Cubic>]}}})
     Z=&FloatFromCurve(FloatCurve=DataInterface() {propertyValues {Curve {keys [<at 0 0.8 RCIM_Cubic>,<at 0.5 1.2 RCIM_Cubic>]}}})
    ))
   &ScaleColor(Scale Alpha=&Lerp_Float(A=0, B=1, Alpha=$User.FuelFlicker), ScaleA=true, ScaleRGB=true) {module=ScaleColor001}
  ]
  rendererClasses [&NiagaraMeshRendererProperties]
  renderers [{rendererClass=&NiagaraMeshRendererProperties, rendererIndex=0}]
 }
 Head {
  emitterUpdateScript [
   &EmitterState()
   &SpawnRate(SpawnRate=250, Use Spawn Probability=true)
  ]
  particleSpawnScript [
   &PartitionParticles()
   &InitializeParticle(
    Lifetime=&MakeCustomFloatFromBool(Bool=$User.IsLocal, True Float=0.15, False Float=0.25)
    Ribbon Width Mode=::NewEnumerator1
    Sprite Size Mode=::NewEnumerator1
    UsePositionOffset=false
    Write Material Random=true
   )
   &SkeletalMeshLocation(
    Bone / Socket Index=$Particles.PartitionParticles.Partition
    MaskSpawnGroup=false
    Mesh Sampling Type=::NewEnumerator5
    Sampled Position Offset=<-3 0 0>
    Skeletal Mesh=DataInterface() {propertyValues {
      DefaultMesh=&M_MED_MurkyVole_Pack
      ExcludeBoneName=root
      OptionalComponentClass=null
      PreviewMesh=&M_MED_MurkyVole_Pack
      SkinningMode=SkinOnTheFly
      SoftSourceActor=null
      SourceMode=Default
      UvSetIndex=0
      WholeMeshLOD=-1
      bExcludeBone=false
      bReadDeformedGeometry=true
      bRequireCurrentFrameData=true
      ComponentTags []
      FilteredBones []
      FilteredSockets [FX_Contrail_L, FX_Contrail_R]
      MeshUserParameter {parameter {name=None, typeDefHandle {registeredTypeIndex=1}, varData []}}
      SamplingRegions []
     }}
    Socket Sampling Mode=::NewEnumerator5
   )
   &AddVelocity(
    Additional Quaternion Rotation=false
    Rotation Mode=::NewEnumerator3
    Rotation Quaternion:Quat4f=$Particles.MeshOrientation
    Velocity=&MakeCustomVectorFromBool(Bool=$User.IsLocal, True Vector=<0 0 10>, False Vector=<0 0 20>)
   )
  ]
  particleUpdateScript [
   &ParticleState()
   &ScaleColor(
    Scale Mode=::NewEnumerator1
    Scale RGBA=&MakeVector4_FromLinearColor(LinearColor=&LerpLinearColors(
      StartColor=&ColorFromCurve(ColorCurve=DataInterface() {propertyValues {
         AlphaCurve {keys [<at 0 1 RCIM_Cubic>,<at 1 0 RCIM_Cubic>]}
         BlueCurve {keys [<at 0 0 RCIM_Cubic>,<at 0.2335 1 RCIM_Cubic>]}
         GreenCurve {keys [<at 0 1.629 RCIM_Cubic>,<at 0.2335 1 RCIM_Cubic>]}
         RedCurve {keys [<at 0 50 RCIM_Cubic>,<at 0.2335 1 RCIM_Cubic>]}
        }})
      EndColor=&ColorFromCurve(ColorCurve=DataInterface() {propertyValues {
         AlphaCurve {keys [<at 0 1 RCIM_Cubic>,<at 1 0 RCIM_Cubic>]}
         BlueCurve {keys [<at 0 0 RCIM_Cubic>,<at 0.3842 1 RCIM_Cubic>]}
         GreenCurve {keys [<at 0 0.2003 RCIM_Cubic>,<at 0.3842 1 RCIM_Cubic>]}
         RedCurve {keys [<at 0 1 RCIM_Cubic>,<at 0.3842 1 RCIM_Cubic>]}
        }})
      LerpFactor=$User.FuelFlicker
     ))
    ScaleRGBA=true
   )
   &Drag(Use Linear Drag=true, Use Rotational Drag=true)
   &InheritSourceMovement(
    Applied Position Scale=&VectorFromFloat(Value=&FloatFromCurve(
      FloatCurve=DataInterface() {propertyValues {Curve {keys [{interpMode=RCIM_Cubic, tangentMode=RCTM_Break, time=0, value=1}, <at 1 0>]}}}
      Scale Curve=0.5
     ))
    ApplyForce=false
    ApplyPosition=true
    ApplyPositionFalloff=false
   )
   &SolveForcesAndVelocity(Clamp Velocity=false, Limit Acceleration=false)
   &ScaleRibbonWidth(Ribbon Width Scale=&FloatFromCurve(FloatCurve=DataInterface() {propertyValues {Curve {keys [{interpMode=RCIM_Cubic, tangentMode=RCTM_Break, time=0, value=0}, <at 0.5 1 RCIM_Cubic>, <at 1 0>]}}}))
   &DynamicMaterialParameters(
    Index 0 Param 1=&FloatFromCurve(FloatCurve=DataInterface() {propertyValues {Curve {keys [{interpMode=RCIM_Cubic, tangentMode=RCTM_Break, time=0, value=1}, {interpMode=RCIM_Cubic, tangentMode=RCTM_Break, time=1, value=0}]}}})
    Index 0 Param 3=&SelectFloatFromArray(Direct Array Index=$Particles.PartitionParticles.Partition, Float Selection Array=DataInterface() {propertyValues {FloatData [-1.5,-1]}})
    Index 0 Param 4=0.25
    Module.Index0Param0Write=true
    Module.Index0Param1Write=true
    Module.Index0Param2Write=true
    Module.Index0Param3Write=true
   )
   &ScaleColor(
    Scale Alpha=&Lerp_Float(
     A=&Waveform("Waveform [1]"=::NewEnumerator8, "[1] Frequency"=1, "[1] Amplitude Min/Max"=<0.125 1>)
     B=1
     Alpha=$User.FuelFlicker
    )
    ScaleA=true
    ScaleRGB=true
   ) {module=ScaleColor001}
   &ScaleColor(
    Scale Alpha=&MakeCustomFloatFromBool(Bool=$User.IsLocal, True Float=1.5, False Float=2)
    ScaleA=true
    ScaleRGB=true
   ) {module=ScaleColor002}
  ]
  rendererClasses [&NiagaraSpriteRendererProperties, &NiagaraRibbonRendererProperties]
  renderers [
   {rendererClass=&NiagaraSpriteRendererProperties}
   {rendererClass=&NiagaraRibbonRendererProperties, rendererIndex=1}
  ]
 }
 Sparks {
  emitterUpdateScript [
   &EmitterState()
   &SpawnRate(SpawnRate=175, Use Spawn Probability=false)
  ]
  particleSpawnScript [
   &PartitionParticles()
   &InitializeParticle(
    Color Mode=::NewEnumerator0
    Lifetime Max=0.15
    Lifetime Min=0.1
    Lifetime Mode=::NewEnumerator1
    Sprite Size Mode=::NewEnumerator1
    Uniform Sprite Size Max=5
    Uniform Sprite Size Min=4
    UsePositionOffset=false
    Write Material Random=true
   )
   &SkeletalMeshLocation(
    Bone / Socket Index=$Particles.PartitionParticles.Partition
    Interpolate Mesh Sampling=false
    MaskSpawnGroup=false
    Mesh Sampling Type=::NewEnumerator5
    Sampled Position Offset=<7 0 0>
    Skeletal Mesh=DataInterface() {propertyValues {
      DefaultMesh=&M_MED_MurkyVole_Pack
      ExcludeBoneName=root
      OptionalComponentClass=null
      PreviewMesh=&M_MED_MurkyVole_Pack
      SkinningMode=SkinOnTheFly
      SoftSourceActor=null
      SourceMode=Default
      UvSetIndex=0
      WholeMeshLOD=-1
      bExcludeBone=false
      bReadDeformedGeometry=true
      bRequireCurrentFrameData=true
      ComponentTags []
      FilteredBones []
      FilteredSockets [Socket_JetThruster_Left, Socket_JetThruster_Right]
      MeshUserParameter {parameter {name=None, typeDefHandle {registeredTypeIndex=1}, varData []}}
      SamplingRegions []
     }}
    Socket Sampling Mode=::NewEnumerator5
   )
   &AddVelocity(
    Additional Quaternion Rotation=false
    Rotation Mode=::NewEnumerator3
    Rotation Quaternion:Quat4f=$Particles.SkeletalMeshLocation.SampledRotation
    Velocity=<1 0 0>
    Velocity Speed Scale=250
   )
   &ShapeLocation(MaskSpawnGroup=false, Override Seed=false, Sphere Radius=5, UseOwnerScale=true)
   &CurlNoiseForce(Noise Strength=125, Pan Noise=false)
   &ApplyInitialForces()
   &InheritVelocity(Inherited Velocity Amount Scale=<-0.1 -0.1 -0.1>, Inherited Velocity Speed Limit=2500, Speed Limit=false, UseSpeedThreshold=false)
  ]
  particleUpdateScript [
   &SolveForcesAndVelocity(Clamp Velocity=false, Limit Acceleration=false)
   &ParticleState()
   &ScaleSpriteSize()
   &ScaleSpriteSizeBySpeed(Max Scale Factor=<0.25 2.5>, Min Scale Factor=<0.5 1>)
   &ScaleColor(
    Linear Color Curve=DataInterface() {propertyValues {
      AlphaCurve {keys [<at 0 1>,<at 1 0>]}
      BlueCurve {keys [<at 0 0>,<at 1 0>]}
      GreenCurve {keys [<at 0 1.2515>,<at 1 0.0071>]}
      RedCurve {keys [<at 0 5>,<at 1 0.1>]}
     }}
    Scale Mode=::NewEnumerator2
   )
  ]
  rendererClasses [&NiagaraSpriteRendererProperties]
  renderers [{rendererClass=&NiagaraSpriteRendererProperties}]
 }
}
summary {
 systemName=NS_MurkyVole_HoverJetBoost
 emitters [
  {emitterName=Combined_Rate, rendererClasses [&NiagaraMeshRendererProperties]}
  {
   emitterName=Airglow_Light_Flare
   rendererClasses [&NiagaraSpriteRendererProperties, &NiagaraLightRendererProperties]
  }
  {emitterName=Burst, rendererClasses [&NiagaraMeshRendererProperties]}
  {emitterName=Sparks, rendererClasses [&NiagaraSpriteRendererProperties]}
  {emitterName=Combined_Rate001, rendererClasses [&NiagaraMeshRendererProperties]}
  {emitterName=Flame, rendererClasses [&NiagaraMeshRendererProperties]}
  {emitterName=Head, rendererClasses [&NiagaraSpriteRendererProperties, &NiagaraRibbonRendererProperties]}
 ]
 userVariables [
  User.FuelFlicker {type=NiagaraFloat, defaultValue=1, description=""}
  User.IsClient {type=NiagaraBool}
  User.IsLocal {type=NiagaraBool, defaultValue=true}
  User.MuzzlePosition {type=NiagaraPosition, description="", defaultValue {}}
  User.Skeletal Mesh {
   type=NiagaraDataInterfaceSkeletalMesh
   defaultValue {
    struct=NiagaraExt_VariableValue_DataInterface
    value {dataInterface=&self:NiagaraDataInterfaceSkeletalMesh_0}
   }
  }
  User.uPhase {type=NiagaraFloat, defaultValue=1, description=""}
 ]
}
