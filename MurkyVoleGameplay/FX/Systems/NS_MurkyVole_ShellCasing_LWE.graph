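# Niagara system that ejects tumbling mesh shell casings, likely for weapon fire.
# A single Shells emitter spawns an instantaneous burst with a mostly upward, slightly backward (local -X) velocity, pulled down by strong gravity over a fixed lifetime.
# Random rotation rates spin each mesh, and material randoms and dynamic material parameters are written per particle.
# It has no collision module or user parameters, so casings follow a simple ballistic arc.
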NiagaraSystem /MurkyVoleGameplay/FX/Systems/NS_MurkyVole_ShellCasing_LWE
dependencies {usedDataInterfaces [], usedDynamicInputs [], usedModules [&SystemState]}
emitters {Shells {
  emitterUpdateScript [
   &EmitterState()
   &SpawnBurst_Instantaneous(Use Spawn Probability=false, UseLoopCountLimit=false)
  ]
  particleSpawnScript [
   &InitializeParticle(Color Mode=::NewEnumerator0, Lifetime=5, UsePositionOffset=false, Write Material Random=true)
   &AddVelocity(Velocity=<-50 0 250>)
  ]
  particleUpdateScript [
   &GravityForce(Gravity=<0 0 -1500>)
   &SolveForcesAndVelocity(Clamp Velocity=false, Limit Acceleration=false)
   &ParticleState()
   &UpdateMeshOrientation(Rotation Rate=&RandomRangeFloat(Minimum=-7, Maximum=7), Rotation Vector=<1 1 1>)
   &DynamicMaterialParameters(Module.Index0Param0Write=true, Module.Index0Param1Write=true, Module.Index0Param2Write=true, Module.Index0Param3Write=true)
  ]
  rendererClasses [&NiagaraMeshRendererProperties]
  renderers [{rendererClass=&NiagaraMeshRendererProperties, rendererIndex=0}]
 }}
summary {
 systemName=NS_MurkyVole_ShellCasing_LWE
 emitters [{emitterName=Shells, rendererClasses [&NiagaraMeshRendererProperties]}]
 userVariables []
}
