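# Weapon muzzle flash Niagara system that adapts to muzzle attachments and aim-down-sights through system bools set by a bitwise module.
# Muzzle_Core_Elements renders mesh and sprite flashes, adding elements for brake or breacher and rescaling for ADS, suppressor and random scale.
# Ejected_Material multiplexes smoke, sparks and optional heated elements in one sprite emitter using spawn-group masks.
# Lighting adds a brief fading light, skipped when suppressed; LWE is disabled.
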NiagaraSystem /MurkyVoleGameplay/FX/Systems/NS_MuzzleFlash_MurkyVole
dependencies {
 usedDataInterfaces [
  &NiagaraDataInterfaceCurve
  &NiagaraDataInterfaceArrayFloat
  &NiagaraDataInterfaceArrayFloat3
  &NiagaraDataInterfaceVectorCurve
 ]
 usedDynamicInputs []
 usedModules [
  &self:SystemRandomScale
  &self:SetSystemBoolsfromBitwise
  &SystemState
 ]
}
emitters {
 Ejected_Material {
  emitterUpdateScript [
   &EmitterState()
   &SpawnBurst_Instantaneous(Spawn Count=5, Use Spawn Probability=false, UseLoopCountLimit=false)
   &SpawnBurst_Instantaneous(
    Spawn Count=&Multiply_Int(
     A=&Multiply_Int(
      A=&Multiply_Int(A=3, B=&MakeIntFromFloat(Float=0, Conversion Mode=::NewEnumerator2))
      B=&MakeCustomIntFromBool(Bool=$System.isMuzzleSupressor, True INT=2, False INT=1)
     )
     B=&MakeCustomIntFromBool(Bool=&Boolean_OrOperation(Bool1=$System.isMuzzleSupressor, Bool2=$System.isADS), True INT=0, False INT=1)
    )
    Spawn Group=1
    Use Spawn Probability=false
    UseLoopCountLimit=false
   ) {module=SpawnBurst_Instantaneous001}
   &SpawnRate(
    SpawnGroup=1
    SpawnRate=&Multiply_Float(
     A=&Multiply_Float(
      A=&Multiply_Float(A=20, B=&MakeCustomFloatFromBool(Bool=$System.isMuzzleBrake, True Float=2, False Float=1))
      B=&FloatFromCurve(
       FloatCurve=DataInterface() {propertyValues {Curve {keys [{interpMode=RCIM_Cubic, leaveTangentWeight=0.9834, tangentMode=RCTM_User, time=0, value=1}, <at 0.2 0 RCIM_Cubic>]}}}
       CurveIndex=$Emitter.NormalizedLoopAge
      )
     )
     B=&MakeCustomFloatFromBool(Bool=$System.isMuzzleSupressor, True Float=0, False Float=1)
    )
    Use Spawn Probability=false
   )
   &SpawnBurst_Instantaneous(
    Spawn Count=&MakeCustomIntFromBool(
     Bool=$System.isADS
     True INT=0
     False INT=&MakeCustomIntFromBool(Bool=$System.bUseHeatedElements, True INT=1, False INT=0)
    )
    Spawn Group=2
    Use Spawn Probability=false
    UseLoopCountLimit=false
   ) {module=SpawnBurst_Instantaneous002}
   &SpawnBurst_Instantaneous(
    Spawn Count=&MakeCustomIntFromBool(Bool=$System.bUseHeatedElements, True INT=1, False INT=0)
    Spawn Group=3
    Use Spawn Probability=false
    UseLoopCountLimit=false
   ) {module=SpawnBurst_Instantaneous003}
  ]
  particleSpawnScript [
   /SetVariables(Particles.VisibilityTag=&SpawnGroup()) {isSetParameters=true}
   &InitializeParticle(
    Color=&MaskLinearColorBySpawnGroup(
     Masked Linear Color=&self:ADS_Muzzle_Adjusted_LinearColor(
      Base Color:LinearColor=<rgba 0.7326 0.7047 0.7047 0.25>
      ADS_Scale:LinearColor=<rgba 1 1 1 0.01>
      MuzzleSupressor_Scale:LinearColor=<rgba 1.36 1.42 1.42 0.25>
      PerWeaponBP_BaseColorVariance:NiagaraFloat=0.6
      ADS:NiagaraBool=$System.isADS
      MuzzleSupressor:NiagaraBool=$System.isMuzzleSupressor
     )
     Passthrough Linear Color=&MaskLinearColorBySpawnGroup(Masked Linear Color=<rgba 1 1 1 1>, Passthrough Linear Color=<rgba 0.5 0.5 0.5 0.05>, Spawn Group Mask=1)
    )
    Lifetime=&MaskFloatBySpawnGroup(
     Masked Float=&RandomRangeFloat(Minimum=0.3, Maximum=0.7)
     Passthrough Float=&MaskFloatBySpawnGroup(
      Masked Float=&RandomRangeFloat(Minimum=0.1, Maximum=0.3)
      Passthrough Float=&MaskFloatBySpawnGroup(Masked Float=&RandomRangeFloat(Minimum=1, Maximum=1.1), Passthrough Float=0.15, Spawn Group Mask=2)
      Spawn Group Mask=1
     )
    )
    Position Offset=&MakeCustomVectorFromBool(
     Bool=&SetBoolByIntComparison(A=&SpawnGroup(), B=2, Comparison Type=::NewEnumerator2)
     True Vector=<0 0 0>
     False Vector=&Multiply_VectorByFloat(
      Vector=<15 0 0>
      Float=&ClampFloat(Float=&Add_Float(
        A=&ReturnNormalizedExecIndex()
        B=&MakeCustomFloatFromBool(Bool=$System.isADS, True Float=0.0063, False Float=0.05)
       ))
     )
    )
    Position Offset Coordinate Space=::Local
    Sprite Rotation Angle=&MaskFloatBySpawnGroup(
     Masked Float=&RandomRangeFloat(Minimum=0, Maximum=360)
     Passthrough Float=&MaskFloatBySpawnGroup(
      Masked Float=0
      Passthrough Float=&MaskFloatBySpawnGroup(
       Masked Float=&Add_Float(A=30, B=&RandomRangeFloat(Minimum=-15, Maximum=15))
       Passthrough Float=0
       Spawn Group Mask=2
      )
      Spawn Group Mask=1
     )
    )
    Sprite Rotation Mode=::NewEnumerator0
    Sprite Size Mode=::NewEnumerator3
    Uniform Sprite Size=&MaskFloatBySpawnGroup(
     Masked Float=&self:ADS_Muzzle_Adjusted_Float_Single(
      Masked Float Array:NiagaraDataInterfaceArrayFloat=DataInterface() {propertyValues {FloatData [45,65,150,150,100]}}
      ADS_Scale:NiagaraFloat=0.65
      MuzzleSupressor_Scale:NiagaraFloat=1
      PerWeaponBP_Scale:NiagaraFloat=1
      ADS:NiagaraBool=$System.isADS
      MuzzleSupressor:NiagaraBool=$System.isMuzzleSupressor
      System Random Scale:NiagaraFloat=&FloatFromCurve(
       FloatCurve=DataInterface() {propertyValues {Curve {keys [<at 0 0.66 RCIM_Cubic>,<at 0.33 1 RCIM_Cubic>,<at 1 0.66>]}}}
       CurveIndex=&ReturnNormalizedExecIndex()
       Scale Curve=$System.RandomScale
      )
      AmmoType:NiagaraInt32=1
     )
     Passthrough Float=&MaskFloatBySpawnGroup(
      Masked Float=&RandomRangeFloat(Minimum=0.75, Maximum=1.5)
      Passthrough Float=&RandomRangeFloat(Minimum=30, Maximum=50)
      Spawn Group Mask=1
     )
    )
    UsePositionOffset=true
    Write Material Random=true
   )
   &ShapeLocation(
    Hemisphere Distribution=<1 0.5>
    MaskSpawnGroup=true
    Override Seed=false
    Rotation Mode=::NewEnumerator1
    Spawn Group Mask=1
    Sphere Radius=6
    UseOwnerScale=true
   )
   &AddVelocity(
    Cone Angle=&MaskFloatBySpawnGroup(Masked Float=10, Passthrough Float=30)
    Module.UseConeFalloff=false
    Module.UseVelDistribution=false
    Override Seed=false
    Velocity Mode=::NewEnumerator2
    Velocity Speed=&MaskFloatBySpawnGroup(
     Masked Float=&RandomRangeFloat(Minimum=400, Maximum=600)
     Passthrough Float=&MaskFloatBySpawnGroup(Masked Float=&RandomRangeFloat(Minimum=900, Maximum=1200), Passthrough Float=0, Spawn Group Mask=1)
    )
   )
   &CameraOffset(Camera Offset Amount=&MaskFloatBySpawnGroup(Masked Float=10, Passthrough Float=0))
  ]
  particleUpdateScript [
   &ParticleState()
   &AccelerationForce(Acceleration=&MaskVectorBySpawnGroup(
     Masked Vector=<0 0 1500>
     Passthrough Vector=&MaskVectorBySpawnGroup(Masked Vector=<0 0 -980>, Passthrough Vector=<0 0 150>, Spawn Group Mask=1)
    ))
   &VectorNoiseForce(
    Force Amount=&MaskFloatBySpawnGroup(Masked Float=15000, Passthrough Float=0, Spawn Group Mask=1)
    Module.Use Random Noise Time Interval=true
    Random Noise Time Interval=0.05
   )
   &Drag(
    Drag=&MaskFloatBySpawnGroup(
     Masked Float=&Multiply_Float(
      A=&RandomRangeFloat(Minimum=8, Maximum=13)
      B=&FloatFromCurve(FloatCurve=DataInterface() {propertyValues {Curve {keys [{arriveTangent=-5.0888, interpMode=RCIM_Cubic, leaveTangent=-5.0888, tangentMode=RCTM_User, time=0.5, value=1}, <at 1 0 RCIM_Cubic>]}}})
     )
     Passthrough Float=&MaskFloatBySpawnGroup(Masked Float=&RandomRangeFloat(Minimum=1, Maximum=5), Passthrough Float=3, Spawn Group Mask=1)
    )
    Use Linear Drag=true
    Use Rotational Drag=false
   )
   &SolveForcesAndVelocity(Clamp Velocity=false, Limit Acceleration=false)
   &SpriteRotationRate(Rotation Rate=&MaskFloatBySpawnGroup(
     Masked Float=&FloatFromCurve(
      FloatCurve=DataInterface() {propertyValues {Curve {keys [<at -0.0046 1.0113 RCIM_Cubic>,<at 1.0011 0.09 RCIM_Cubic>]}}}
      Scale Curve=&RandomRangeFloat(Minimum=-3, Maximum=3)
     )
     Passthrough Float=0
    ))
   &ScaleSpriteSizeBySpeed(
    Max Scale Factor=&MaskVector2dBySpawnGroup(
     Masked Vector2D=&RandomRangeVector2D(Minimum=<1 2>, Maximum=<1 10>)
     Passthrough Vector2D=<1 1>
     Spawn Group Mask=1
    )
    Min Scale Factor=&MaskVector2dBySpawnGroup(
     Masked Vector2D=&RandomRangeVector2D(Minimum=<0.5 0.5>, Maximum=<1 1>)
     Passthrough Vector2D=<1 1>
     Spawn Group Mask=1
    )
   )
   &ScaleSpriteSize(
    Scale Sprite Size Mode=::NewEnumerator0
    Uniform Scale Factor=&MaskFloatBySpawnGroup(
     Masked Float=&FloatFromCurve(FloatCurve=DataInterface() {propertyValues {Curve {keys [
          {arriveTangent=1.4888, interpMode=RCIM_Cubic, leaveTangent=1.4888, leaveTangentWeight=0.8692, tangentMode=RCTM_User, time=0, value=0.2}
          {arriveTangentWeight=0.6592, interpMode=RCIM_Cubic, tangentMode=RCTM_User, time=1, value=1}
         ]}}})
     Passthrough Float=&MaskFloatBySpawnGroup(
      Masked Float=&FloatFromCurve(
       FloatCurve=DataInterface() {propertyValues {Curve {keys [{tangentMode=RCTM_User, time=0.8, value=1}, {tangentMode=RCTM_User, time=1, value=0}]}}}
       Scale Curve=&RandomRangeFloat(Minimum=0.1, Maximum=1)
      )
      Passthrough Float=&FloatFromCurve(FloatCurve=DataInterface() {propertyValues {Curve {keys [
           {arriveTangent=0.5942, interpMode=RCIM_Cubic, leaveTangent=0.5942, tangentMode=RCTM_User, time=0, value=0.33}
           {interpMode=RCIM_Cubic, leaveTangent=0.6906, time=1, value=1}
          ]}}})
      Spawn Group Mask=1
     )
    )
   )
   &DynamicMaterialParameters(
    Index 0 Param 1=&Add_Float(
     A=&RandomRangeFloat(Minimum=0, Maximum=1)
     B=&FloatFromCurve(FloatCurve=DataInterface() {propertyValues {Curve {keys [{arriveTangent=-1.1823, interpMode=RCIM_Cubic, leaveTangent=-1.1823, tangentMode=RCTM_User, time=0, value=1}, <at 1 0 RCIM_Cubic>]}}})
    )
    Index 0 Param 2=&FloatFromCurve(FloatCurve=DataInterface() {propertyValues {Curve {keys [
         {arriveTangent=0.8564, interpMode=RCIM_Cubic, leaveTangent=0.8564, tangentMode=RCTM_User, time=0.0046, value=-0.9753}
         {arriveTangent=4.2876, interpMode=RCIM_Cubic, leaveTangent=5.5268, tangentMode=RCTM_User, time=1, value=1}
        ]}}})
    Index 0 Param 3=&FloatFromCurve(
     FloatCurve=DataInterface() {propertyValues {Curve {keys [
         {arriveTangent=-15.2571, interpMode=RCIM_Cubic, leaveTangent=-15.2571, tangentMode=RCTM_User, time=0, value=1}
         {tangentMode=RCTM_User, time=0.2, value=0}
        ]}}}
     Scale Curve=$User.SmokeEmissiveAmount
    )
    Index 0 Param 4=&FloatFromCurve(FloatCurve=DataInterface() {propertyValues {Curve {keys [{arriveTangent=-3.2497, interpMode=RCIM_Cubic, leaveTangent=-3.2497, tangentMode=RCTM_User, time=0, value=1}, <at 1 0 RCIM_Cubic>]}}})
    Index 1 Param 1=&FloatFromCurve(FloatCurve=DataInterface() {propertyValues {Curve {keys [{arriveTangent=1.674, interpMode=RCIM_Cubic, leaveTangent=1.674, tangentMode=RCTM_User, time=0, value=0.5}, <at 1 1 RCIM_Cubic>]}}})
    Index 1 Param 2=&RandomRangeFloat(Minimum=0, Maximum=1)
    Index 1 Param 3=&FloatFromCurve(FloatCurve=DataInterface() {propertyValues {Curve {keys [<at 0 10 RCIM_Cubic>, {arriveTangent=-85.6983, interpMode=RCIM_Cubic, leaveTangent=-85.6983, tangentMode=RCTM_User, time=0.25, value=1}]}}})
    Index 1 Param 4=&RandomRangeFloat(Minimum=2, Maximum=6)
    Index 2 Param 1=&FloatFromCurve(FloatCurve=DataInterface() {propertyValues {Curve {keys [
         {interpMode=RCIM_Cubic, leaveTangent=1.673, tangentMode=RCTM_Break, time=0, value=0.5}
         {interpMode=RCIM_Cubic, tangentMode=RCTM_Break, time=1, value=1}
        ]}}})
    Index 2 Param 2=&FloatFromCurve(
     FloatCurve=DataInterface() {propertyValues {Curve {keys [
         {interpMode=RCIM_Cubic, leaveTangentWeight=0.397, tangentMode=RCTM_Break, time=0, value=0.7}
         {arriveTangentWeight=0.4342, interpMode=RCIM_Cubic, tangentMode=RCTM_Break, time=1, value=1}
        ]}}}
     Scale Curve=&RandomRangeFloat(Minimum=0, Maximum=1)
    )
    Index 2 Param 3=&FloatFromCurve(
     FloatCurve=DataInterface() {propertyValues {Curve {keys [
         {interpMode=RCIM_Cubic, leaveTangent=2.6489, leaveTangentWeight=0.3345, tangentMode=RCTM_Break, time=0, value=0}
         {arriveTangentWeight=0.3308, interpMode=RCIM_Cubic, tangentMode=RCTM_Break, time=1, value=1}
        ]}}}
     Scale Curve=5
    )
    Index 2 Param 4=&FloatFromCurve(FloatCurve=DataInterface() {propertyValues {Curve {keys [
         {interpMode=RCIM_Cubic, leaveTangent=0.5158, leaveTangentWeight=0.3333, tangentMode=RCTM_Break, time=0, value=0.8}
         {arriveTangentWeight=0.3592, interpMode=RCIM_Cubic, tangentMode=RCTM_Break, time=1, value=1}
        ]}}})
    Module.Param0WriteEnabled=true
    Module.Param0WriteEnabled1=true
    Module.Param0WriteEnabled2=true
    Module.Param0WriteEnabled3=true
    Module.Param1WriteEnabled=true
    Module.Param1WriteEnabled1=true
    Module.Param1WriteEnabled2=true
    Module.Param1WriteEnabled3=true
    Module.Param2WriteEnabled=true
    Module.Param2WriteEnabled1=true
    Module.Param2WriteEnabled2=true
    Module.Param2WriteEnabled3=false
    Module.Param3WriteEnabled=true
    Module.Param3WriteEnabled1=true
    Module.Param3WriteEnabled2=true
    Module.Param3WriteEnabled3=false
    Write Parameter Index 1=true
    Write Parameter Index 2=true
    Write Parameter Index 3=true
   )
   &ScaleColor(
    Scale Alpha=&MaskFloatBySpawnGroup(
     Masked Float=&FloatFromCurve(FloatCurve=DataInterface() {propertyValues {Curve {keys [
          {arriveTangent=9.0725, interpMode=RCIM_Cubic, leaveTangent=9.0725, tangentMode=RCTM_User, time=0, value=0}
          {arriveTangentWeight=0.1274, interpMode=RCIM_Cubic, leaveTangentWeight=0.7778, tangentMode=RCTM_User, time=0.1408, value=0.9847}
          {arriveTangent=-0.0405, interpMode=RCIM_Cubic, leaveTangent=-0.0405, tangentMode=RCTM_User, time=1, value=0}
         ]}}})
     Passthrough Float=&MaskFloatBySpawnGroup(
      Masked Float=&FloatFromCurve(FloatCurve=DataInterface() {propertyValues {Curve {keys [<at 0 3 RCIM_Cubic>,<at 1 1 RCIM_Cubic>]}}})
      Passthrough Float=&FloatFromCurve(FloatCurve=DataInterface() {propertyValues {Curve {keys [
           {interpMode=RCIM_Cubic, tangentMode=RCTM_User, time=0, value=0}
           {interpMode=RCIM_Cubic, tangentMode=RCTM_User, time=0.2, value=0.9847}
           {interpMode=RCIM_Cubic, tangentMode=RCTM_User, time=0.66, value=1}
           {arriveTangent=-0.0405, interpMode=RCIM_Cubic, leaveTangent=-0.0405, tangentMode=RCTM_User, time=1, value=0}
          ]}}})
      Spawn Group Mask=1
     )
    )
    Scale RGB=&MaskVectorBySpawnGroup(
     Masked Vector=&Multiply_VectorByFloat(
      Vector=<1 1 1>
      Float=&FloatFromCurve(FloatCurve=DataInterface() {propertyValues {Curve {keys [<at 0 1>,<at 1 0>]}}})
     )
     Passthrough Vector=<1 1 1>
    )
    ScaleA=true
    ScaleRGB=true
   )
   /SetVariables(Particles.Sparks Dynamic Parameter:Vector4f=&MakeVector4(X=2, Y=20, Z=90, W=1)) {isSetParameters=true}
  ]
  rendererClasses [&NiagaraSpriteRendererProperties]
  renderers [
   {rendererClass=&NiagaraSpriteRendererProperties}
   {rendererClass=&NiagaraSpriteRendererProperties, rendererIndex=1}
   {rendererClass=&NiagaraSpriteRendererProperties, rendererIndex=2}
  ]
 }
 LWE {bEnabled=false, rendererClasses [], renderers []}
 Lighting {
  emitterUpdateScript [
   &EmitterState()
   &SpawnBurst_Instantaneous(
    Spawn Count=&Multiply_Int(A=1, B=&MakeCustomIntFromBool(Bool=$System.isMuzzleSupressor, True INT=0, False INT=1))
    Use Spawn Probability=false
    UseLoopCountLimit=false
   )
  ]
  particleSpawnScript [&InitializeParticle(
    Lifetime=0.1
    Position Offset=<7 0 0>
    Sprite Size Mode=::NewEnumerator3
    Uniform Sprite Size=15
    UsePositionOffset=true
    Write Material Random=true
   )]
  particleUpdateScript [
   &ParticleState()
   &ScaleColor(
    Scale Alpha=&Lerp_Float(
     A=20
     B=0
     Alpha=&FloatFromCurve(FloatCurve=DataInterface() {propertyValues {Curve {keys [<at 0 1>,<at 1 0>]}}})
    )
    Scale RGB=&MakeVectorFromLinearColorRGB(Color=&self:Material_Hue_Shift(Color:LinearColor=<rgba 1 0.3636 0.01 1>, Hue Shift:NiagaraFloat=0))
    ScaleA=true
    ScaleRGB=true
   )
   /SetVariables(
    Particles.LightExponent=5
    Particles.LightRadius=150
    Particles.PointLightIntensity:NiagaraFloat=&Multiply_Float(
     A=1500
     B=&FloatFromCurve(FloatCurve=DataInterface() {propertyValues {Curve {keys [<at 0 1>,<at 1 0>]}}})
    )
   ) {isSetParameters=true}
   /SetVariables(Particles.Position=&AddVectorToPosition(
     Position=$Particles.Position
     Vector=&TransformVector(Vector=&Lerp_Vector(A=<7 0 0>, B=<70 0 0>, Alpha=$Particles.NormalizedAge), Destination Space=::Simulation)
    )) {enabled=false, isSetParameters=true}
  ]
  rendererClasses [&NiagaraLightRendererProperties, &NiagaraComponentRendererProperties]
  renderers [
   {rendererClass=&NiagaraLightRendererProperties, rendererIndex=0}
   {rendererClass=&NiagaraComponentRendererProperties, rendererIndex=1}
  ]
 }
 Muzzle_Core_Elements {
  emitterUpdateScript [
   &EmitterState()
   &SpawnBurst_Instantaneous(Use Spawn Probability=false, UseLoopCountLimit=false)
   &SpawnBurst_Instantaneous(
    Spawn Count=&MakeCustomIntFromBool(
     Bool=$System.isMuzzleBreacher
     True INT=6
     False INT=&MakeCustomIntFromBool(Bool=$System.isMuzzleBrake, True INT=2, False INT=0)
    )
    Spawn Group=1
    Use Spawn Probability=false
    UseLoopCountLimit=false
   ) {module=SpawnBurst_Instantaneous001}
  ]
  particleSpawnScript [
   /SetVariables(Particles.VisibilityTag=&SpawnGroup()) {isSetParameters=true}
   &PartitionParticles(Number Of Partitions=&MaskIntBySpawnGroup(Masked Int=0, Passthrough Int=6, Spawn Group Mask=0))
   &InitializeParticle(
    Color=<rgba 3.5 1.4395 0 1>
    Lifetime=&RandomRangeFloat(Minimum=0.075, Maximum=0.1)
    Mesh Scale=&Multiply_VectorByFloat(
     Vector=&self:ADS_Muzzle_Adjusted_Vector(
      Masked Vector Array:NiagaraDataInterfaceArrayFloat3=DataInterface() {propertyValues {FloatData [<0.5 0.25 0.25>, <0.875 0.375 0.375>, <2.625 1.125 1.125>, <1 1 1>, <1.75 0.75 0.75>]}}
      Pass through Array:NiagaraDataInterfaceArrayFloat3=DataInterface() {propertyValues {FloatData [<0.4375 0.1875 0.1875>, <0.65 0.25 0.25>, <1.25 0.6 0.6>, <0.5 0.5 0.5>, <0.875 0.375 0.375>]}}
      ADS Scale:Vector3f=&MaskVectorBySpawnGroup(Masked Vector=<1 0.5 0.5>, Passthrough Vector=<0.5 0.25 0.25>)
      MuzzleSupressor_Scale:Vector3f=&MaskVectorBySpawnGroup(Masked Vector=<0.5 0.4 0.4>, Passthrough Vector=<0.6 0.3 0.3>)
      PerWeaponBP_Scale:Vector3f=<1 1 1>
      ADS:NiagaraBool=$System.isADS
      MuzzleSupressor:NiagaraBool=$System.isMuzzleSupressor
      System Random Scale:NiagaraFloat=$System.RandomScale
      AmmoType:NiagaraInt32=1
     )
     Float=$System.RandomScale
    )
    Mesh Scale Mode=::NewEnumerator4
    Position Offset=&MaskVectorBySpawnGroup(
     Masked Vector=<-3 0 0>
     Passthrough Vector=&MakeCustomVectorFromBool(Bool=$System.isMuzzleBrake, True Vector=<0 0 1>, False Vector=<0 0 0>)
    )
    Sprite Rotation Mode=::NewEnumerator2
    Sprite Size=&Vector2DFromFloat(Value=&Multiply_Float(
      A=&self:ADS_Muzzle_Adjusted_Float(
       Masked Float Array:NiagaraDataInterfaceArrayFloat=DataInterface() {propertyValues {FloatData [50,60,120,100,80]}}
       Pass through Float Array:NiagaraDataInterfaceArrayFloat=DataInterface() {propertyValues {FloatData [30,40,60,50,40]}}
       ADS_Scale:NiagaraFloat=0.5
       MuzzleSupressor_Scale:NiagaraFloat=0
       PerWeaponBP_Scale:NiagaraFloat=1
       ADS:NiagaraBool=$System.isADS
       MuzzleSupressor:NiagaraBool=$System.isMuzzleSupressor
       System Random Scale:NiagaraFloat=$System.RandomScale
       AmmoType:NiagaraInt32=1
      )
      B=$System.RandomScale
     ))
    Sprite Size Mode=::NewEnumerator4
    UsePositionOffset=true
    Write Material Random=true
   )
   &InitialMeshOrientation(
    Mesh Orientation Mode=::NewEnumerator2
    Orientation Vector=&MaskVectorBySpawnGroup(
     Masked Vector=<1 0 0>
     Passthrough Vector=&self:Rotation(
      Default:Vector3f=<0 1 0>
      float:NiagaraFloat=&Divide_Float(A=&ReturnNormalizedExecIndex(Normalized Index Includes One=false), B=1)
     )
    )
    Rotation=&RandomRangeVector(Minimum=<0 0 0>, Maximum=<1 0 0>)
    Use Rotation Vector=true
   )
   &CameraOffset(Camera Offset Amount=1)
   &SpriteFacingAndAlignment(
    Sprite Alignment=&MaskVectorBySpawnGroup(Masked Vector=<0 1 0>, Passthrough Vector=<1 0 0>)
    Sprite Facing=&MaskVectorBySpawnGroup(
     Masked Vector=<1 0 0>
     Passthrough Vector=&Lerp_Vector(A=<0 1 0>, B=<0 -1 0>, Alpha=&MakeFloatFromInt(Int=$Particles.PartitionParticles.Partition))
    )
    WriteAlignment=true
    WriteFacing=true
   )
   /SetVariables(Particles.SpriteCameraOffset=5) {isSetParameters=true}
  ]
  particleUpdateScript [
   &ParticleState()
   &ScaleSpriteSize(
    Scale Factor=&Vector2DFromFloat(Value=&FloatFromCurve(FloatCurve=DataInterface() {propertyValues {Curve {keys [{tangentMode=RCTM_User, time=0, value=0}, {tangentMode=RCTM_User, time=1, value=1}]}}}))
    Scale Sprite Size Mode=::NewEnumerator2
   )
   &ScaleMeshSize(Scale Factor=&VectorFromCurve(VectorCurve=DataInterface() {propertyValues {XCurve {keys [<at 0 0>,<at 1 1>]}, YCurve {keys [<at 0 0.25>,<at 1 1.25>]}, ZCurve {keys [<at 0 0.25>,<at 1 1.25>]}}}))
   &ScaleColor(
    Scale Alpha=&RandomRangeFloat(Minimum=0.5, Maximum=1)
    Scale RGB=&MaskVectorBySpawnGroup(
     Masked Vector=&VectorFromFloat(Value=&FloatFromCurve(FloatCurve=DataInterface() {propertyValues {Curve {keys [{arriveTangent=-42.9916, interpMode=RCIM_Cubic, leaveTangent=-42.9916, tangentMode=RCTM_User, time=0, value=5}, <at 0.15 1 RCIM_Cubic>]}}}))
     Passthrough Vector=<1 1 1>
    )
    ScaleA=true
    ScaleRGB=true
   )
   &DynamicMaterialParameters(
    Index 0 Param 1=&FloatFromCurve(
     FloatCurve=DataInterface() {propertyValues {Curve {keys [
         {arriveTangent=2.9073, interpMode=RCIM_Cubic, leaveTangent=2.9073, leaveTangentWeight=0.3077, tangentMode=RCTM_User, time=0.35, value=0}
         {arriveTangentWeight=0.8354, interpMode=RCIM_Cubic, tangentMode=RCTM_User, time=1, value=1}
        ]}}}
     Scale Curve=0.75
    )
    Index 0 Param 2=&FloatFromCurve(FloatCurve=DataInterface() {propertyValues {Curve {keys [
         {interpMode=RCIM_Cubic, leaveTangentWeight=0.3627, tangentMode=RCTM_User, time=0.4, value=1}
         {arriveTangent=-1.799, arriveTangentWeight=1.0345, interpMode=RCIM_Cubic, leaveTangent=-1.799, tangentMode=RCTM_User, time=1, value=0}
        ]}}})
    Index 0 Param 3=&MakeCustomFloatFromBool(Bool=&RandomBool(), True Float=1, False Float=0)
    Module.Index0Param0Write=true
    Module.Index0Param1Write=true
    Module.Index0Param2Write=true
    Module.Index0Param3Write=false
   )
   /SetVariables(Particles.SpritePOsition:NiagaraPosition=&AddVectorToPosition(Position=$Particles.Position, Vector=<0 0 0>)) {isSetParameters=true}
  ]
  rendererClasses [&NiagaraMeshRendererProperties, &NiagaraSpriteRendererProperties]
  renderers [
   {rendererClass=&NiagaraMeshRendererProperties, rendererIndex=0}
   {rendererClass=&NiagaraSpriteRendererProperties, rendererIndex=1}
   {rendererClass=&NiagaraMeshRendererProperties, rendererIndex=2}
   {rendererClass=&NiagaraSpriteRendererProperties, rendererIndex=3}
  ]
 }
}
summary {
 systemName=NS_MuzzleFlash_MurkyVole
 emitters [
  {emitterName=Ejected_Material, rendererClasses [&NiagaraSpriteRendererProperties]}
  {
   emitterName=Muzzle_Core_Elements
   rendererClasses [&NiagaraMeshRendererProperties, &NiagaraSpriteRendererProperties]
  }
  {emitterName=Lighting, rendererClasses [&NiagaraLightRendererProperties, &NiagaraComponentRendererProperties]}
  {bEnabled=false, emitterName=LWE, rendererClasses []}
 ]
 userVariables [
  User.BurstLoops {type=NiagaraInt32, defaultValue=1}
  User.FiringRate {type=NiagaraFloat, defaultValue=0.1, description=""}
  User.MuzzleColorHueControl {type=NiagaraFloat, defaultValue=0, description=""}
  User.MuzzleModsBitSet {type=NiagaraInt32}
  User.SmokeEmissiveAmount {type=NiagaraFloat, defaultValue=0, description=""}
  User.SmokeInitialColor {type=LinearColor, description="", defaultValue {struct=LinearColor, value=<rgba 0.7326 0.7047 0.7047 0.25>}}
 ]
}
